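# Sound cue for the Lonelies music, looping the BunkerJam music pack track under the emote music sound class.
# A single looping wave player feeds the output, using the Lonelies music attenuation asset.
# Overridden concurrency allows one instance at a time, stopping the oldest when a new one starts.
# A submix send also routes it to the Fortnite music audio analysis submix.
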SoundCue /Lonelies/Audio/Lonelies_Music_Cue
root=WavePlayer_1
nodes [{id=WavePlayer_1, properties {bLooping=true, soundWaveAssetPtr=/Game/Athena/Sounds/MusicPacks/MusicPack_BunkerJam}}]
properties {
 attenuationSettings=&/Lonelies_Music_Att
 bOverrideConcurrency=true
 maxDistance=6512
 soundClassObject=&Fort_Music_Emote
 volumeMultiplier=1
 attenuationOverrides {bAttenuate=false, bSpatialize=false}
 concurrencyOverrides {maxCount=1, resolutionRule=StopOldest}
 soundSubmixSends [{minSendLevel=0, soundSubmix=&Fortnite_AudioAnalysis_Music}]
}
